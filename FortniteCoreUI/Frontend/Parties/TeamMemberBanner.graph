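# Party slot banner widget, based on FortCampaignTeamMemberBanner, that shows one team member's connection state.
# Connection state changes switch between an open slot, a connecting throbber, and the connected player banner with leader crown.
# Hovering the open slot's transparent button rebuilds the image brush with the hovered texture.
# Clicking it broadcasts a sidebar open request through the global event router, gated by a dual-blade side menu check.
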WidgetBlueprint /FortniteCoreUI/Frontend/Parties/TeamMemberBanner
graphs [
 EventGraph {nodes [
   "On Button Hovered (ButtonPartyFinder)"!() -> (OutputDelegate, then, Button:Common Button Legacy Object Reference) {is=title K2Node_ComponentBoundEvent}
   > HandleMouseHoverVisualState!(execute, self, Hover:bool=true) -> (then)
   "On Button Unhovered (ButtonPartyFinder)"!() -> (OutputDelegate, then, Button:Common Button Legacy Object Reference) {is=title K2Node_ComponentBoundEvent}
   > HandleMouseHoverVisualState(Hover=false)
   "On Button Clicked (ButtonPartyFinder)"!() -> (OutputDelegate, then, Button:Common Button Legacy Object Reference) {is=title K2Node_ComponentBoundEvent}
   > OpenPartyFinder()
   n1 = Event OnTeamMemberConnectionStateChanged!() -> (OutputDelegate, then, NewConnectionState:EFortMemberConnectionState Enum=Open) {is=title K2Node_Event}
   > SetActiveWidget(
    self:Widget Switcher Object Reference=Get Switcher_MemberState() -> (Switcher_MemberState:Common Widget Switcher Legacy Object Reference)
    Widget=Select!(
     Open:Widget Object Reference=Get Open() -> (Open:Overlay Object Reference)
     Connecting:Widget Object Reference=Get Connecting() -> (Connecting:Size Box Object Reference)
     Connected:Widget Object Reference=Get Connected() -> (Connected:Overlay Object Reference)
     Invalid:Widget Object Reference=Get Open()
     Index:EFortMemberConnectionState Enum=@n1.NewConnectionState
    ) -> (ReturnValue:Widget Object Reference)
   )
  ]}
 HandleMouseHoverVisualState {nodes [
   n1 = HandleMouseHoverVisualState!() -> (then, Hover:bool=~) {is=title K2Node_FunctionEntry}
   n2 = Break Slate Brush!(SlateBrush:"Slate Brush Structure (by ref)"=Get Brush!(self:Image Object Reference=Get OpenImage() -> (OpenImage:Image Object Reference)) -> (Brush:Slate Brush Structure)) -> (
    TintColor:Slate Color Structure
    DrawAs:ESlateBrushDrawType Enum
    Tiling:ESlateBrushTileType Enum
    Mirroring:ESlateBrushMirrorType Enum
    ImageSize:Vector 2D Structure
    Margin:Margin Structure
    ResourceObject:Object Reference
    OutlineSettings:Slate Brush Outline Settings Structure
   ) {is=title K2Node_BreakStruct}
   > SetBrush(
    self=Get OpenImage()
    InBrush=Make Slate Brush!(
     TintColor:Slate Color Structure=@n2
     DrawAs:ESlateBrushDrawType Enum=@n2.DrawAs
     Tiling:ESlateBrushTileType Enum=@n2.Tiling
     Mirroring:ESlateBrushMirrorType Enum=@n2.Mirroring
     ImageSize:Vector 2D Structure=@n2.ImageSize
     Margin:Margin Structure=@n2.Margin
     ResourceObject:Object Reference=Select!(
      Option 0:Object Reference=/FortniteCoreUI/UI/Foundation/Textures/Buttons/T-Slot-Banner-Empty-S-128
      Option 1:Object Reference=/FortniteCoreUI/UI/Foundation/Textures/Buttons/T-Slot-Banner-Empty-Hovered-S-128
      Index:bool=@n1
     ) -> (ReturnValue:Object Reference)
     OutlineSettings:Slate Brush Outline Settings Structure=(CornerRadii=(X=0,Y=0,Z=0,W=0),Color=(SpecifiedColor=(R=0,G=0,B=0,A=0),ColorUseRule=UseColor_Specified),Width=0,RoundingType=HalfHeightRadius,bUseBrushTransparency=False)
    ) -> (SlateBrush:Slate Brush Structure) {is=title K2Node_MakeStruct}
   )
  ]}
 OpenPartyFinder {nodes [
   OpenPartyFinder() {is=title K2Node_FunctionEntry}
   > n1 = ShouldUseDualBladeSideMenu!(execute) -> (then, ReturnValue:bool)
   > Branch(Condition=@n1)
   > Broadcast Event!(
    execute
    self:Gameplay Event Router Component Object Reference=GetGlobalEventRouter()
    EventData:Sidebar Request Open Structure=Make Sidebar Request Open!() -> (SidebarRequestOpen:Sidebar Request Open Structure) {is=title K2Node_MakeStruct}
    EventContext:Object Reference=GetOwningPlayer()
   ) -> (then, ReturnValue:bool)
  ]}
]
variables []
widgets {
 info {inheritedWidgetCount=3, parentClass=/Script/FortniteUI.FortCampaignTeamMemberBanner, rootWidgetClass=&Overlay, widgetCount=11}
 widgets [
  Overlay_17 {widgetClassPath=&Overlay, widget=Overlay_17}
  Switcher_MemberState {
   widgetClassPath=&CommonWidgetSwitcherLegacy
   bIsVariable=true
   parent=Overlay_17
   widget=Switcher_MemberState
   slot {class=OverlaySlot}
  }
  Open {widgetClassPath=&Overlay, bIsVariable=true, parent=Switcher_MemberState, widget=Open, slot {class=WidgetSwitcherSlot}}
  OpenImage {
   widgetClassPath=&Image
   bIsVariable=true
   parent=Open
   widget=OpenImage
   properties {
    visibility=SelfHitTestInvisible
    brush {
     bIsDynamicallyLoaded=false
     drawAs=Image
     imageType=NoImage
     mirroring=NoMirror
     resourceName=None
     resourceObject=&T-Slot-Banner-Empty-S-128
     tiling=NoTile
     imageSize=<40 48>
     margin=<ltrb 0 0 0 0>
     outlineSettings {
      bUseBrushTransparency=false
      roundingType=HalfHeightRadius
      width=0
      color {colorUseRule=UseColor_Specified, specifiedColor=<rgba 0 0 0 0>}
      cornerRadii=<0 0 0 0>
     }
     tintColor {colorUseRule=UseColor_Specified, specifiedColor=<rgba 1 1 1 1>}
     uVRegion {bIsValid=false, max=<0 0>, min=<0 0>}
    }
   }
   slot {class=OverlaySlot}
  }
  ButtonPartyFinder {
   widgetClassPath=&IconTextButton
   bIsVariable=true
   parent=Open
   widget=ButtonPartyFinder
   properties {
    buttonText=""
    onButtonBaseClicked=(&self:WidgetTree.ButtonPartyFinder.HandleOnButtonClicked)
    onButtonBaseDoubleClicked=(&self:WidgetTree.ButtonPartyFinder.HandleOnButtonDoubleClicked)
    onButtonBaseHovered=(&self:WidgetTree.ButtonPartyFinder.HandleOnButtonHovered)
    onButtonBaseUnhovered=(&self:WidgetTree.ButtonPartyFinder.HandleOnButtonUnhovered)
    onSelectedChangedBase=(&self:WidgetTree.ButtonPartyFinder.HandleOnSelectedChanged)
    style=/FortUILibrary/Foundation/Buttons/ButtonStyles/ButtonStyle-MediumTransparentNoCues.ButtonStyle-MediumTransparentNoCues_C
   }
   slot {class=OverlaySlot, properties {horizontalAlignment=HAlign_Fill, verticalAlignment=VAlign_Fill}}
  }
  Connecting {
   widgetClassPath=&SizeBox
   bIsVariable=true
   parent=Switcher_MemberState
   widget=Connecting
   properties {bOverride_HeightOverride=true, bOverride_WidthOverride=true, heightOverride=40, widthOverride=40}
   slot {class=WidgetSwitcherSlot}
  }
  Image_571 {
   widgetClassPath=&Image
   bIsVariable=true
   parent=Connecting
   widget=Image_571
   properties {
    visibility=SelfHitTestInvisible
    brush {
     bIsDynamicallyLoaded=false
     drawAs=Image
     imageType=NoImage
     mirroring=NoMirror
     resourceName=None
     resourceObject=/FortUILibrary/UIKit/Throbber/MI_UIKit_Monochrome_Throbber
     tiling=NoTile
     imageSize=<16 16>
     margin=<ltrb 0 0 0 0>
     outlineSettings {
      bUseBrushTransparency=false
      roundingType=HalfHeightRadius
      width=0
      color {colorUseRule=UseColor_Specified, specifiedColor=<rgba 0 0 0 0>}
      cornerRadii=<0 0 0 0>
     }
     tintColor {colorUseRule=UseColor_Specified, specifiedColor=<rgba 1 1 1 1>}
     uVRegion {bIsValid=false, max=<0 0>, min=<0 0>}
    }
   }
   slot {class=SizeBoxSlot}
  }
  Connected {widgetClassPath=&Overlay, bIsVariable=true, parent=Switcher_MemberState, widget=Connected, slot {class=WidgetSwitcherSlot}}
  Banner_MemberBanner {
   widgetClassPath=/Script/FortniteUI.FortPlayerBanner
   bInherited=true
   bIsVariable=true
   parent=Connected
   widget=Banner_MemberBanner
   properties {desiredSize=<40 48>}
   slot {class=OverlaySlot, properties {horizontalAlignment=HAlign_Fill, verticalAlignment=VAlign_Fill}}
  }
  Image_LeaderIcon {
   widgetClassPath=&Image
   bInherited=true
   bIsVariable=true
   parent=Connected
   widget=Image_LeaderIcon
   properties {
    visibility=SelfHitTestInvisible
    brush {
     bIsDynamicallyLoaded=false
     drawAs=Image
     imageType=NoImage
     mirroring=NoMirror
     resourceName=None
     resourceObject=/Game/UI/Foundation/Banner/M-CrownShine
     tiling=NoTile
     imageSize=<16 16>
     margin=<ltrb 0 0 0 0>
     outlineSettings {
      bUseBrushTransparency=false
      roundingType=HalfHeightRadius
      width=0
      color {colorUseRule=UseColor_Specified, specifiedColor=<rgba 0 0 0 0>}
      cornerRadii=<0 0 0 0>
     }
     tintColor {colorUseRule=UseColor_Specified, specifiedColor=<rgba 1 1 1 1>}
     uVRegion {bIsValid=false, max=<0 0>, min=<0 0>}
    }
    renderTransform {angle=0, scale=<1 1>, shear=<0 0>, translation=<0 -8>}
   }
   slot {class=OverlaySlot, properties {horizontalAlignment=HAlign_Center}}
  }
  InputActionWidget {widgetClassPath=&CommonActionWidget, bInherited=true, widget=null}
 ]
}
