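# Modal panel for managing a party member, showing their name, power rating and banner beside action buttons inherited from FortTeamMemberManage.
# A ContentSwitcher swaps between these options and the local player banner editor, activating or deactivating the editor as it changes.
# OpenBannerEditor shows the editor, while closing it or activating the panel restores the options.
# The graph also relabels the Sit Out/Jump In button on participation changes and pops the panel on Close.
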WidgetBlueprint /Game/UI/Frontend/Parties/TeamMemberManage
graphs [EventGraph {nodes [
   "Close Banner Editor (BP_LocalPlayerBannerEditor)"!() -> (OutputDelegate, then, IconId:Name=None, ColorId:Name=None) {is=title K2Node_ComponentBoundEvent}
   > SetActiveWidget(
    self:Widget Switcher Object Reference=Get ContentSwitcher() -> (ContentSwitcher:Common Widget Switcher Legacy Object Reference)
    Widget=Get TeamMemberManageOptions() -> (TeamMemberManageOptions:Size Box Object Reference)
   )
   > OnBannerEditorClosed(self:Fort Team Member Manage Object Reference)
   n1 = "On Active Widget Changed (ContentSwitcher)"!() -> (OutputDelegate, then, ActiveWidget:Widget Object Reference, ActiveWidgetIndex:int=0) {is=title K2Node_ComponentBoundEvent}
   > n2 = Branch(Condition="Equal (Object)"(A=@n1.ActiveWidget, B=Get BP_LocalPlayerBannerEditor() -> (BP_LocalPlayerBannerEditor:WBP Local Player Banner Editor Object Reference)) {is=title K2Node_CallFunction})
   > OnWidgetActivated(self:WBP Local Player Banner Editor Object Reference=Get BP_LocalPlayerBannerEditor())
   OnWidgetDeactivated(execute=@n2.else, self:WBP Local Player Banner Editor Object Reference=Get BP_LocalPlayerBannerEditor())
   n3 = "On Clicked (CloseProfileButton)"!() -> (OutputDelegate, then, Button:Common Button Base Object Reference) {is=title K2Node_ComponentBoundEvent}
   > Fire Event Element Interacted(Element=@n3.Button)
   > PopActivatablePanelInModalLayer!(execute, Panel:Common Activatable Panel Legacy Object Reference=Self-Reference()) -> (then)
   n4 = Event OnParticipationChanged!() -> (OutputDelegate, then, bIsParticipating:bool) {is=title K2Node_Event}
   > n5 = Cast To WBP_UIKit_ButtonCTA_Base!(execute, Object=Get Button_ToggleParticipation() -> (Button_ToggleParticipation:Common Button Base Object Reference)) -> (then, CastFailed, AsWBP UIKit Button CTA Base:WBP UIKit Button CTA Base Object Reference) {is=title K2Node_DynamicCast}
   > SetText!(
    execute
    self:Fort CTAButton Object Reference=@n5
    ButtonText:"Text (by ref)"=Select(Option 0:Text="LOCTEXT(\"Jump In!\")", Option 1:Text="LOCTEXT(\"Sit Out\")", Index=@n4.bIsParticipating) -> (ReturnValue:Text)
   ) -> (then)
   Event On Activated()
   > SetActiveWidget(self=Get ContentSwitcher(), Widget=Get TeamMemberManageOptions())
   Event On Deactivated()
   > OnWidgetDeactivated(self=Get BP_LocalPlayerBannerEditor())
   Event OpenBannerEditor() {is=title K2Node_Event}
   > SetActiveWidget(self=Get ContentSwitcher(), Widget=Get BP_LocalPlayerBannerEditor())
  ]}]
variables []
widgets {
 info {inheritedWidgetCount=15, parentClass=/Script/FortniteUI.FortTeamMemberManage, rootWidgetClass=&Overlay, widgetCount=28}
 widgets [
  Overlay_15 {widgetClassPath=&Overlay, widget=Overlay_15}
  WBP_UIKit_Scrim {
   widgetClassPath=/FortUILibrary/Systems/Dialog/WBP_UIKit_Scrim
   bIsVariable=true
   parent=Overlay_15
   widget=WBP_UIKit_Scrim
   slot {class=OverlaySlot, properties {horizontalAlignment=HAlign_Fill, verticalAlignment=VAlign_Fill}}
  }
  GridPanel_0 {
   widgetClassPath=&GridPanel
   parent=Overlay_15
   widget=GridPanel_0
   properties {columnFill [1], rowFill [1]}
   slot {class=OverlaySlot, properties {horizontalAlignment=HAlign_Center, verticalAlignment=VAlign_Center}}
  }
  WBP_UIKit_Backplate {
   widgetClassPath=&WBP_UIKit_Backplate
   bIsVariable=true
   parent=GridPanel_0
   widget=WBP_UIKit_Backplate
   properties {brightness=Dark}
   slot {class=GridSlot}
  }
  ContentSwitcher {
   widgetClassPath=&CommonWidgetSwitcherLegacy
   bIsVariable=true
   parent=GridPanel_0
   widget=ContentSwitcher
   properties {transitionDuration=0}
   slot {class=GridSlot, properties {layer=1, padding=<ltrb 32 32 32 32>}}
  }
  TeamMemberManageOptions {
   widgetClassPath=&SizeBox
   bIsVariable=true
   parent=ContentSwitcher
   widget=TeamMemberManageOptions
   properties {bOverride_MinDesiredWidth=true, heightOverride=512, minDesiredWidth=724, widthOverride=720}
   slot {class=WidgetSwitcherSlot}
  }
  VerticalBox_108 {widgetClassPath=&VerticalBox, parent=TeamMemberManageOptions, widget=VerticalBox_108, slot {class=SizeBoxSlot}}
  CommonTextBlock_0 {
   widgetClassPath=&CommonTextBlock
   parent=VerticalBox_108
   widget=CommonTextBlock_0
   properties {
    justification=Center
    scrollStyle=&TextScrollStyle-Base_C
    style=/FortUILibrary/Styles/Typography/_TextStyle_SubH2._TextStyle_SubH2_C
    text=MANAGE TEAM MEMBER
    font {
     bForceMonospaced=false
     bMaterialIsStencil=false
     fontMaterial=null
     fontObject=&Burbank
     letterSpacing=0
     monospacedWidth=1
     size=33
     skewAmount=0
     typefaceFontName=BurbankRegular-Black
     outlineSettings {
      bApplyOutlineToDropShadows=false
      bMiteredCorners=false
      bSeparateFillAlpha=false
      outlineMaterial=null
      outlineSize=0
      outlineColor=<rgba 0 0 0 1>
     }
    }
    shadowOffset=<0 0>
    strikeBrush {
     bIsDynamicallyLoaded=false
     drawAs=NoDrawType
     imageType=NoImage
     mirroring=NoMirror
     resourceName=None
     resourceObject=null
     tiling=NoTile
     imageSize=<32 32>
     margin=<ltrb 0 0 0 0>
     outlineSettings {
      bUseBrushTransparency=false
      roundingType=HalfHeightRadius
      width=0
      color {colorUseRule=UseColor_Specified, specifiedColor=<rgba 0 0 0 0>}
      cornerRadii=<0 0 0 0>
     }
     tintColor {colorUseRule=UseColor_Specified, specifiedColor=<rgba 1 1 1 1>}
     uVRegion {bIsValid=false, max=<0 0>, min=<0 0>}
    }
   }
   slot {class=VerticalBoxSlot, properties {verticalAlignment=VAlign_Top, padding=<ltrb 0 0 0 32>}}
  }
  HorizontalBox_285 {
   widgetClassPath=&HorizontalBox
   parent=VerticalBox_108
   widget=HorizontalBox_285
   slot {class=VerticalBoxSlot, properties {size {sizeRule=Fill, value=1}}}
  }
  VerticalBoxNameAndBanner {widgetClassPath=&VerticalBox, parent=HorizontalBox_285, widget=VerticalBoxNameAndBanner, slot {class=HorizontalBoxSlot}}
  Spacer_242 {
   widgetClassPath=&Spacer
   parent=VerticalBoxNameAndBanner
   widget=Spacer_242
   properties {size=<320 0>}
   slot {class=VerticalBoxSlot}
  }
  Text_PlayerName {
   widgetClassPath=&CommonTextBlock
   bInherited=true
   bIsVariable=true
   parent=VerticalBoxNameAndBanner
   widget=Text_PlayerName
   properties {
    justification=Center
    style=/FortUILibrary/Styles/Typography/_TextStyle_TagNormal._TextStyle_TagNormal_C
    text=Text Block
    font {
     bForceMonospaced=false
     bMaterialIsStencil=false
     fontMaterial=null
     fontObject=&BurbankSmall
     letterSpacing=0
     monospacedWidth=1
     size=15
     skewAmount=0
     typefaceFontName=Black
     outlineSettings {
      bApplyOutlineToDropShadows=false
      bMiteredCorners=false
      bSeparateFillAlpha=false
      outlineMaterial=null
      outlineSize=0
      outlineColor=<rgba 0 0 0 1>
     }
    }
    shadowOffset=<0 0>
    strikeBrush {
     bIsDynamicallyLoaded=false
     drawAs=NoDrawType
     imageType=NoImage
     mirroring=NoMirror
     resourceName=None
     resourceObject=null
     tiling=NoTile
     imageSize=<32 32>
     margin=<ltrb 0 0 0 0>
     outlineSettings {
      bUseBrushTransparency=false
      roundingType=HalfHeightRadius
      width=0
      color {colorUseRule=UseColor_Specified, specifiedColor=<rgba 0 0 0 0>}
      cornerRadii=<0 0 0 0>
     }
     tintColor {colorUseRule=UseColor_Specified, specifiedColor=<rgba 1 1 1 1>}
     uVRegion {bIsValid=false, max=<0 0>, min=<0 0>}
    }
   }
   slot {class=VerticalBoxSlot, properties {horizontalAlignment=HAlign_Center, verticalAlignment=VAlign_Bottom, padding=<ltrb 0 12 0 0>}}
  }
  PowerRatingWidgetSwitcher {
   widgetClassPath=&CommonWidgetSwitcherLegacy
   bInherited=true
   bIsVariable=true
   parent=VerticalBoxNameAndBanner
   widget=PowerRatingWidgetSwitcher
   slot {class=VerticalBoxSlot, properties {horizontalAlignment=HAlign_Center}}
  }
  PowerRating_MemberPowerRating {
   widgetClassPath=/Game/UI/Foundation/Stats/PowerRatingWidget
   bInherited=true
   bIsVariable=true
   parent=PowerRatingWidgetSwitcher
   widget=PowerRating_MemberPowerRating
   properties {showDualPower=false, showProgress=false, useLargeStyle=false}
   slot {class=WidgetSwitcherSlot}
  }
  Moonbeam_PowerRating_MemberPowerRating {
   widgetClassPath=/Game/UI/Foundation/Stats/Moonbeam_PowerRatingWidget
   bInherited=true
   bIsVariable=true
   parent=PowerRatingWidgetSwitcher
   widget=Moonbeam_PowerRating_MemberPowerRating
   properties {showDualPower=false, showProgress=false, useLargeStyle=false}
   slot {class=WidgetSwitcherSlot}
  }
  Banner_PlayerBanner {
   widgetClassPath=/Script/FortniteUI.FortPlayerBanner
   bInherited=true
   bIsVariable=true
   parent=VerticalBoxNameAndBanner
   widget=Banner_PlayerBanner
   properties {visibility=HitTestInvisible, desiredSize=<128 158>}
   slot {class=VerticalBoxSlot, properties {horizontalAlignment=HAlign_Center, verticalAlignment=VAlign_Top, padding=<ltrb 0 12 0 0>}}
  }
  VBox_Buttons {
   widgetClassPath=&VerticalBox
   bInherited=true
   bIsVariable=true
   parent=HorizontalBox_285
   widget=VBox_Buttons
   properties {navigation=&self:WidgetTree.VBox_Buttons.WidgetNavigation_0}
   slot {class=HorizontalBoxSlot, properties {verticalAlignment=VAlign_Center, padding=<ltrb 0 0 88 0>, size {sizeRule=Fill, value=1}}}
  }
  Button_EditBanner {
   widgetClassPath=&WBP_UIKit_Button_Regular
   bInherited=true
   bIsVariable=true
   parent=VBox_Buttons
   widget=Button_EditBanner
   properties {minWidth=312, text=Edit Banner}
   slot {class=VerticalBoxSlot, properties {verticalAlignment=VAlign_Top, padding=<ltrb 0 0 0 24>}}
  }
  Button_MakePartyLeader {
   widgetClassPath=&WBP_UIKit_Button_Regular
   bInherited=true
   bIsVariable=true
   parent=VBox_Buttons
   widget=Button_MakePartyLeader
   properties {minWidth=312, text=Make Party Leader}
   slot {class=VerticalBoxSlot, properties {verticalAlignment=VAlign_Top, padding=<ltrb 0 0 0 24>}}
  }
  Button_Whisper {
   widgetClassPath=&WBP_UIKit_Button_Regular
   bInherited=true
   bIsVariable=true
   parent=VBox_Buttons
   widget=Button_Whisper
   properties {minWidth=312, text=Whisper}
   slot {class=VerticalBoxSlot, properties {verticalAlignment=VAlign_Top, padding=<ltrb 0 0 0 24>}}
  }
  Button_Kick {
   widgetClassPath=&WBP_UIKit_Button_Regular
   bInherited=true
   bIsVariable=true
   parent=VBox_Buttons
   widget=Button_Kick
   properties {minWidth=312, text=Kick}
   slot {class=VerticalBoxSlot, properties {verticalAlignment=VAlign_Top, padding=<ltrb 0 0 0 24>}}
  }
  Button_ViewProfile {
   widgetClassPath=&WBP_UIKit_Button_Regular
   bInherited=true
   bIsVariable=true
   parent=VBox_Buttons
   widget=Button_ViewProfile
   properties {minWidth=312, text=View Profile}
   slot {class=VerticalBoxSlot, properties {verticalAlignment=VAlign_Top, padding=<ltrb 0 0 0 24>}}
  }
  Button_Mute {
   widgetClassPath=&WBP_UIKit_Button_Regular
   bInherited=true
   bIsVariable=true
   parent=VBox_Buttons
   widget=Button_Mute
   properties {minWidth=312, text=Mute}
   slot {class=VerticalBoxSlot, properties {verticalAlignment=VAlign_Top, padding=<ltrb 0 0 0 24>}}
  }
  Button_Unmute {
   widgetClassPath=&WBP_UIKit_Button_Regular
   bInherited=true
   bIsVariable=true
   parent=VBox_Buttons
   widget=Button_Unmute
   properties {minWidth=312, text=Unmute}
   slot {class=VerticalBoxSlot, properties {verticalAlignment=VAlign_Top, padding=<ltrb 0 0 0 24>}}
  }
  Button_LeaveParty {
   widgetClassPath=&WBP_UIKit_Button_Regular
   bInherited=true
   bIsVariable=true
   parent=VBox_Buttons
   widget=Button_LeaveParty
   properties {minWidth=312, text=Leave Party}
   slot {class=VerticalBoxSlot, properties {verticalAlignment=VAlign_Top, padding=<ltrb 0 0 0 24>}}
  }
  Button_ToggleParticipation {
   widgetClassPath=&WBP_UIKit_Button_Regular
   bInherited=true
   bIsVariable=true
   parent=VBox_Buttons
   widget=Button_ToggleParticipation
   properties {minWidth=312, text=Sit Out}
   slot {class=VerticalBoxSlot, properties {verticalAlignment=VAlign_Top, padding=<ltrb 0 0 0 24>}}
  }
  CloseProfileButton {
   widgetClassPath=&WBP_UIKit_Button_Regular
   bIsVariable=true
   parent=VBox_Buttons
   widget=CloseProfileButton
   properties {minWidth=312, text=Close, triggeringInputAction {dataTable=&InputActionDataTable, rowName=Close}}
   slot {class=VerticalBoxSlot, properties {verticalAlignment=VAlign_Top}}
  }
  BP_LocalPlayerBannerEditor {
   widgetClassPath=/Game/UI/Foundation/Profile/WBP_LocalPlayerBannerEditor
   bIsVariable=true
   parent=ContentSwitcher
   widget=BP_LocalPlayerBannerEditor
   slot {class=WidgetSwitcherSlot, properties {horizontalAlignment=HAlign_Left, verticalAlignment=VAlign_Top}}
  }
 ]
}
